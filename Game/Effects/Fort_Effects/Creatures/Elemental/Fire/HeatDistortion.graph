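# Unlit translucent particle material for heat distortion in the fire elemental creature effects, usable on sprites and Niagara mesh particles, with index-of-refraction mode.
# A sphere height map is inverted and shifted by particle red to look up a gradient texture, letting particle color animate the mask.
# That mask, multiplied by sphere height and particle alpha, blends a value from neutral 1 to a lower index.
# A sphere normal map is also sampled.
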Material /Game/Effects/Fort_Effects/Creatures/Elemental/Fire/HeatDistortion
graph {
 properties {
  bUsedWithNiagaraMeshParticles=true
  bUsedWithNiagaraSprites=true
  bUsedWithParticleSprites=true
  blendMode=BLEND_Translucent
  pixelDepthOffsetMode=PDOM_Legacy
  refractionMethod=RM_IndexOfRefraction
  shadingModel=MSM_Unlit
  usedShadingModels=Unlit
 }
 nodes [
  Constant()
  n1 = ParticleColor()
  n2 = TextureSample() {samplerType=SAMPLERTYPE_LinearColor, texture=/Engine/Functions/Engine_MaterialFunctions02/ExampleContent/Textures/SphereRenderHeightMap}
  n3 = TextureSample(UVs=Subtract(A=OneMinus(None=@n2.R), B=@n1.R)) {texture=/Game/Global/Textures/Default/Masks_and_Gradients/gradientPass}
  LinearInterpolate(B=Constant() {r=0.8}, Alpha=Multiply(A=@n1.A, B=Multiply(A=@n2.R, B=@n3.G))) {constA=1}
  TextureSample() {samplerType=SAMPLERTYPE_Normal, texture=/Engine/Functions/Engine_MaterialFunctions02/ExampleContent/Textures/SphereRenderNormalsMap}
  Time()
 ]
}
